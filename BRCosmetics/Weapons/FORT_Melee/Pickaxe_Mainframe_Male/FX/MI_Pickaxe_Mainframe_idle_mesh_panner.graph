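# Translucent unlit material for Niagara mesh particles, a fast panning streak effect for the Mainframe pickaxe's idle FX.
# Two smoke trail textures scroll along U: one masked by a sine falloff across V, the other tiled and power-adjusted, then summed and clamped.
# Particle color and alpha scale that mask for emissive and opacity, with optional time-of-day tinting and shared emissive adjustments on the color.
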Material /BRCosmetics/Weapons/FORT_Melee/Pickaxe_Mainframe_Male/FX/MI_Pickaxe_Mainframe_idle_mesh_panner
graph {
 properties {
  bUsedWithNiagaraMeshParticles=true
  blendMode=BLEND_Translucent
  pixelDepthOffsetMode=PDOM_Legacy
  shadingModel=MSM_Unlit
  usedShadingModels=Unlit
 }
 nodes [
  n1 = Constant() {r=1}
  n2 = MaterialFunctionCall(
   Color_Morning=VectorParameter() {parameterName=Color_morning, defaultValue=<rgba 1 1 1 1>}.RGB
   Color_Day=VectorParameter() {parameterName=Color_Day, defaultValue=<rgba 1 1 1 1>}.RGB
   Color_Evening=VectorParameter() {parameterName=Color_Evening, defaultValue=<rgba 1 1 1 1>}.RGB
   Color_Night=VectorParameter() {parameterName=Color_Night, defaultValue=<rgba 1 1 1 1>}.RGB
  ) {materialFunction=&Color_Time_of_Day}
  n3 = ScalarParameter() {parameterName=Use_color_time_of_day}
  n4 = ParticleColor()
  n5 = ComponentMask(None=TextureCoordinate()) {g=true}
  n6 = Saturate(None=Power(Base=Sine(None=@n5) {period=2}))
  n7 = TextureSample(UVs=Panner(Coordinate=TextureCoordinate()) {speedX=10}) {texture=/Game/Effects/Fort_Effects/Textures/Smoke/T_Smoke_Trail_Soft_2}
  n8 = TextureSample(UVs=Panner(Coordinate=TextureCoordinate() {uTiling=4}) {speedX=10}) {samplerType=SAMPLERTYPE_Grayscale, texture=/Game/Effects/Fort_Effects/Textures/Smoke/T_Smoke_Trail_Soft_NoClamp}
  n9 = ScalarParameter() {defaultValue=1.3, parameterName=Power}
  n10 = Clamp(None=Add(A=Multiply(A=@n6, B=@n7.R), B=Power(Base=@n8.R, Exp=@n9) {constExponent=1.3}))
  MaterialFunctionCall(Emissive=Multiply(A=LinearInterpolate(A=@n1, B=@n2.Result, Alpha=@n3), B=Multiply(A=@n4.RGB, B=@n10))) {materialFunction=&MF_FX_EmissiveAdjustments}
  Reroute(None=Multiply(A=@n4.A, B=@n10))
 ]
}
